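# Control Rig for the Scholar Ghoul backpack cosmetic that adds procedural secondary motion to the pack and its two mouth jaws.
# Three Chain Harmonics units run in sequence from BeginExecution, rooted at the pack bone, the large jaw hinge and the small jaw hinge.
# Each layers pendulum, reach and noisy wave motion; the jaws use zero stiffness with downward gravity, stronger on the large one.
# No unit is rooted at the ribbon bones, and all three leave debug drawing enabled.
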ControlRigBlueprint /BRCosmetics/Accessories/FORT_Backpacks/Backpack_F_MED_Scholar_Ghoul/Meshes/F_MED_Scholar_Ghoul_Pack_CtrlRig
graphs [
 RigVMFunctionLibrary {links [], nodes []}
 RigVMModel {nodes [
   RigUnit_BeginExecution = &RigUnit_BeginExecution()
   /Script/ControlRig.RigUnit_ChainHarmonicsPerItem(
    ChainRoot=(Type=Bone,Name="pack")
    DrawWorldOffset=(Rotation=(X=0,Y=0,Z=0,W=1),Translation=(X=0,Y=0,Z=0),Scale3D=(X=1,Y=1,Z=1))
    Pendulum=(bEnabled=True,PendulumStiffness=2,PendulumGravity=(X=0,Y=0,Z=0),PendulumBlend=0.75,PendulumDrag=0.98,PendulumMaximum=0.1,UnwindAxis=(X=0,Y=1,Z=0),UnwindMinimum=0.2,UnwindMaximum=0.05)
    Reach=(bEnabled=True,ReachTarget=(X=0,Y=0,Z=0),ReachAxis=(X=1,Y=0,Z=0))
    Speed=(X=0,Y=2,Z=0.8)
    Wave=(bEnabled=True,WaveFrequency=(X=0,Y=0,Z=0),WaveAmplitude=(X=0,Y=0.1,Z=0.1),WaveOffset=(X=0,Y=0,Z=0),WaveNoise=(X=0,Y=1,Z=3),WaveMinimum=1,WaveMaximum=2,WaveEase=SineEaseInOut)
    WaveCurve=(EditorCurveData=(Keys=((),(Time=1,Value=1)),DefaultValue=340282346638528859811704183484516925440,PreInfinityExtrap=RCCE_Constant,PostInfinityExtrap=RCCE_Constant))
    bDrawDebug=True
    ExecutePin=@RigUnit_BeginExecution.ExecutePin
   ) {is=struct RigVMUnitNode}
   > &RigUnit_ChainHarmonicsPerItem(
    ChainRoot=(Type=Bone,Name="dyn_large_mouth_jaw_hinge_jnt")
    DrawWorldOffset=(Rotation=(X=0,Y=0,Z=0,W=1),Translation=(X=0,Y=0,Z=0),Scale3D=(X=1,Y=1,Z=1))
    Pendulum=(bEnabled=True,PendulumStiffness=0,PendulumGravity=(X=0,Y=0,Z=-65),PendulumBlend=0.75,PendulumDrag=0.98,PendulumMaximum=0.1,UnwindAxis=(X=0,Y=1,Z=0),UnwindMinimum=0.2,UnwindMaximum=0.05)
    Reach=(bEnabled=True,ReachTarget=(X=0,Y=0,Z=0),ReachAxis=(X=1,Y=0,Z=0),ReachMinimum=0,ReachMaximum=0)
    Speed=(X=0,Y=1.5,Z=1)
    Wave=(bEnabled=True,WaveFrequency=(X=0,Y=0,Z=0),WaveAmplitude=(X=0,Y=0.7,Z=0.1),WaveOffset=(X=0,Y=0,Z=0),WaveNoise=(X=0,Y=0.5,Z=1),WaveMinimum=0,WaveMaximum=-0.5,WaveEase=SineEaseIn)
    WaveCurve=(EditorCurveData=(Keys=((),(Time=1,Value=1)),DefaultValue=340282346638528859811704183484516925440,PreInfinityExtrap=RCCE_Constant,PostInfinityExtrap=RCCE_Constant))
    bDrawDebug=True
   )
   > &RigUnit_ChainHarmonicsPerItem(
    ChainRoot=(Type=Bone,Name="dyn_small_mouth_jaw_hinge_jnt")
    DrawWorldOffset=(Rotation=(X=0,Y=0,Z=0,W=1),Translation=(X=0,Y=0,Z=0),Scale3D=(X=1,Y=1,Z=1))
    Pendulum=(bEnabled=True,PendulumStiffness=0,PendulumGravity=(X=0,Y=0,Z=-35),PendulumBlend=0.75,PendulumDrag=0.98,PendulumMaximum=0.1,UnwindAxis=(X=0,Y=1,Z=0),UnwindMinimum=0.2,UnwindMaximum=0.05)
    Reach=(bEnabled=True,ReachTarget=(X=0,Y=0,Z=0),ReachAxis=(X=1,Y=0,Z=0))
    Speed=(X=0,Y=2,Z=2)
    Wave=(bEnabled=True,WaveFrequency=(X=0,Y=0,Z=0),WaveAmplitude=(X=0,Y=0.2,Z=0.02),WaveOffset=(X=0,Y=0,Z=0),WaveNoise=(X=0,Y=0.5,Z=1),WaveMinimum=0,WaveMaximum=-1,WaveEase=SineEaseInOut)
    WaveCurve=(EditorCurveData=(Keys=((),(Time=1,Value=1)),DefaultValue=340282346638528859811704183484516925440,PreInfinityExtrap=RCCE_Constant,PostInfinityExtrap=RCCE_Constant))
    bDrawDebug=True
   )
  ]}
]
hierarchy [
 root {loc=<0 0 0>, rot=<0 0 0 1>}
 pack {parent=root, loc=<0 -0.7316 0>, rot=<0 0 0 1>}
 dyn_ribbons_master_jnt {parent=pack, loc=<0 -3.0964 0>, rot=<0 0 0 1>}
 dyn_large_mouth_jnt {parent=dyn_ribbons_master_jnt, loc=<-2.6618 -3.2152 -0.0517>, rot=<0 0 0 1>}
 dyn_large_mouth_jaw_hinge_jnt {parent=dyn_large_mouth_jnt, loc=<-1.1537 0.0034 -0.151>, rot=<0 0 0 1>}
 dyn_large_mouth_jaw_jnt {parent=dyn_large_mouth_jaw_hinge_jnt, loc=<-1.2587 -0.0034 -0.1476>, rot=<0 0 0 1>}
 dyn_small_mouth_jnt {parent=dyn_ribbons_master_jnt, loc=<2.7194 -3.2152 -0.0187>, rot=<0 0 0 1>}
 dyn_small_mouth_jaw_hinge_jnt {parent=dyn_small_mouth_jnt, loc=<0.5048 0.0034 -0.1841>, rot=<0 0 0 1>}
 dyn_small_mouth_jaw_jnt {parent=dyn_small_mouth_jaw_hinge_jnt, loc=<0.5872 -0.0034 -0.1476>, rot=<0 0 0 1>}
 dyn_ribbon1_1_jnt {parent=dyn_ribbons_master_jnt, loc=<-0.8715 0.1521 -2.6596>, rot=<-0.0166 -0.7573 0.019 -0.6526>}
 dyn_ribbon1_2_jnt {parent=dyn_ribbon1_1_jnt, loc=<3.1311 -0.0341 -0.109>, rot=<0.0756 -0.052 -0.0087 -0.9957>}
 dyn_ribbon1_3_jnt {parent=dyn_ribbon1_2_jnt, loc=<4.0781 -0.1446 -0.0047>, rot=<0.0021 0.0057 0.0526 -0.9986>}
 dyn_ribbon1_4_jnt {parent=dyn_ribbon1_3_jnt, loc=<4.685 0.2728 -0.0961>, rot=<0.0188 -0.0524 -0.0193 -0.9983>}
 dyn_ribbon2_1_jnt {parent=dyn_ribbons_master_jnt, loc=<0.9061 -0.087 -2.652>, rot=<0.7602 -0.0581 -0.6459 -0.0376>}
 dyn_ribbon2_2_jnt {parent=dyn_ribbon2_1_jnt, loc=<3.0799 0.0622 -0.0164>, rot=<0.0353 -0.0231 0.0117 -0.999>}
 dyn_ribbon2_3_jnt {parent=dyn_ribbon2_2_jnt, loc=<4.1759 0.0534 -0.0066>, rot=<-0.0228 -0.0114 0.0017 -0.9997>}
 dyn_ribbon2_4_jnt {parent=dyn_ribbon2_3_jnt, loc=<5.1387 0.1416 -0.1315>, rot=<0.0073 -0.0372 -0.0555 -0.9977>}
]
