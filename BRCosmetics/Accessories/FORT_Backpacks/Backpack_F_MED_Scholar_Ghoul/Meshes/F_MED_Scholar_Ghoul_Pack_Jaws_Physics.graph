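# Physics asset for the jaws of the Scholar Ghoul backpack cosmetic, defining simulation bodies for its large and small mouths.
# Five single-capsule bodies cover a ribbons master joint, both mouth joints and both jaw joints.
# Constraints form two chains, linking each mouth to the ribbons master and each jaw to its mouth.
# Angular limits keep twist and first swing tight while allowing a much wider second swing, giving a mostly hinge-like flap.
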PhysicsAsset /BRCosmetics/Accessories/FORT_Backpacks/Backpack_F_MED_Scholar_Ghoul/Meshes/F_MED_Scholar_Ghoul_Pack_Jaws_Physics
bodies {
 dyn_large_mouth_jaw_jnt [{
   shapeType=Capsule
   length=12.25
   radius=3.75
   shapeName=dyn_large_mouth_jaw_jnt_capsule
   center=<-9.3159 0.0002 -2.3072>
   rotation=<rot 80.0024 179.9998 179.9995>
  }]
 dyn_large_mouth_jnt [{
   shapeType=Capsule
   length=11.5
   radius=4.75
   shapeName=dyn_large_mouth_jnt_capsule
   center=<-9.3772 0.0003 5.1035>
   rotation=<rot 60.0001 0 0>
  }]
 dyn_ribbons_master_jnt [{shapeType=Capsule, length=2, radius=2, shapeName=dyn_ribbons_master_jnt_capsule}]
 dyn_small_mouth_jaw_jnt [{
   shapeType=Capsule
   length=8
   radius=3.75
   shapeName=dyn_large_mouth_jaw_jnt_capsule
   center=<7.2097 0 -2.8675>
   rotation=<rot 69.9998 0.0002 0>
  }]
 dyn_small_mouth_jnt [{
   shapeType=Capsule
   length=8.25
   radius=4.25
   shapeName=dyn_large_mouth_jnt_capsule
   center=<7.7825 0.0003 2.8753>
   rotation=<rot -70.0005 0 0>
  }]
}
constraints [
 {bone1Name=dyn_small_mouth_jaw_jnt, bone2Name=dyn_small_mouth_jnt, swing1LimitDegrees=1, swing2LimitDegrees=25, twistLimitDegrees=5}
 {bone1Name=dyn_large_mouth_jnt, bone2Name=dyn_ribbons_master_jnt, swing1LimitDegrees=5, swing2LimitDegrees=25, twistLimitDegrees=5}
 {bone1Name=dyn_large_mouth_jaw_jnt, bone2Name=dyn_large_mouth_jnt, swing1LimitDegrees=1, swing2LimitDegrees=25, twistLimitDegrees=5}
 {bone1Name=dyn_small_mouth_jnt, bone2Name=dyn_ribbons_master_jnt, swing1LimitDegrees=5, swing2LimitDegrees=30, twistLimitDegrees=5}
]
