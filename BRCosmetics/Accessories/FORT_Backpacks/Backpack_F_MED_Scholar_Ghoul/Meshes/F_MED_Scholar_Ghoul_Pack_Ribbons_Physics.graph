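# Physics asset for the ribbons on the Scholar Ghoul backpack cosmetic, defining their collision bodies and joint limits.
# Two chains of four capsules branch from a master ribbon capsule, which connects to a sphere body on the pack through an angularly free constraint.
# Ribbon joints allow wide swing on one axis but almost none on the other and zero twist, so each ribbon bends mostly in one plane.
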PhysicsAsset /BRCosmetics/Accessories/FORT_Backpacks/Backpack_F_MED_Scholar_Ghoul/Meshes/F_MED_Scholar_Ghoul_Pack_Ribbons_Physics
bodies {
 dyn_ribbon1_1_jnt [{shapeType=Capsule, length=1.75, radius=1.5, shapeName=dyn_ribbon1_1_jnt_capsule, center=<2.6751 0 0>, rotation=<rot 90 180 180>}]
 dyn_ribbon1_2_jnt [{shapeType=Capsule, length=2.5, radius=1.5, shapeName=dyn_ribbon1_2_jnt_capsule, center=<3.1217 0 0>, rotation=<rot 90 -144.7356 -144.7355>}]
 dyn_ribbon1_3_jnt [{shapeType=Capsule, length=3, radius=1.5, shapeName=dyn_ribbon1_3_jnt_capsule, center=<3.4609 0 0>, rotation=<rot 90 0 0>}]
 dyn_ribbon1_4_jnt [{shapeType=Capsule, length=1.25, radius=2.75, shapeName=dyn_ribbon1_4_jnt_capsule, center=<3.3982 0.0001 0>, rotation=<rot 90 -0.0001 0>}]
 dyn_ribbon2_1_jnt [{shapeType=Capsule, length=1.75, radius=1.5, shapeName=dyn_ribbon2_1_jnt_capsule, center=<2.6751 0 0>, rotation=<rot 90 0.0066 0.0066>}]
 dyn_ribbon2_2_jnt [{shapeType=Capsule, length=2.5, radius=1.5, shapeName=dyn_ribbon2_2_jnt_capsule, center=<3.1217 0 0>, rotation=<rot 90 0 0>}]
 dyn_ribbon2_3_jnt [{
   shapeType=Capsule
   length=3.5
   radius=1.5
   shapeName=dyn_ribbon2_3_jnt_capsule
   center=<3.579 0.0001 0>
   rotation=<rot 90 -179.9998 -179.9999>
  }]
 dyn_ribbon2_4_jnt [{shapeType=Capsule, length=1.25, radius=2.75, shapeName=dyn_ribbon2_4_jnt_capsule, center=<3.3982 0.0001 0>, rotation=<rot 90 0.0099 0.0099>}]
 dyn_ribbons_master_jnt [{shapeType=Capsule, length=2, radius=2, shapeName=dyn_ribbons_master_jnt_capsule}]
 pack [{shapeType=Sphere, radius=12.7199, shapeName=pack_sphere, center=<-0.2305 11.6081 -8.6751>}]
}
constraints [
 {
  bone1Name=dyn_ribbons_master_jnt
  bone2Name=pack
  swing1LimitDegrees=45
  swing1Motion=Free
  swing2LimitDegrees=45
  swing2Motion=Free
  twistLimitDegrees=45
  twistMotion=Free
 }
 {bone1Name=dyn_ribbon2_4_jnt, bone2Name=dyn_ribbon2_3_jnt, swing1LimitDegrees=45, swing2LimitDegrees=2, twistLimitDegrees=0}
 {bone1Name=dyn_ribbon1_1_jnt, bone2Name=dyn_ribbons_master_jnt, swing1LimitDegrees=45, swing2LimitDegrees=2, twistLimitDegrees=0}
 {bone1Name=dyn_ribbon1_2_jnt, bone2Name=dyn_ribbon1_1_jnt, swing1LimitDegrees=45, swing2LimitDegrees=2, twistLimitDegrees=0}
 {bone1Name=dyn_ribbon1_3_jnt, bone2Name=dyn_ribbon1_2_jnt, swing1LimitDegrees=45, swing2LimitDegrees=2, twistLimitDegrees=0}
 {bone1Name=dyn_ribbon1_4_jnt, bone2Name=dyn_ribbon1_3_jnt, swing1LimitDegrees=45, swing2LimitDegrees=2, twistLimitDegrees=0}
 {bone1Name=dyn_ribbon2_1_jnt, bone2Name=dyn_ribbons_master_jnt, swing1LimitDegrees=45, swing2LimitDegrees=2, twistLimitDegrees=0}
 {bone1Name=dyn_ribbon2_2_jnt, bone2Name=dyn_ribbon2_1_jnt, swing1LimitDegrees=45, swing2LimitDegrees=2, twistLimitDegrees=0}
 {bone1Name=dyn_ribbon2_3_jnt, bone2Name=dyn_ribbon2_2_jnt, swing1LimitDegrees=45, swing2LimitDegrees=2, twistLimitDegrees=0}
]
