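# Anim Blueprint for the Scholar Ghoul back bling, animating its two mouths, eyelids and ribbons procedurally from the ref pose.
# Bone modifiers offset jaws, spread mouths during skydive float and blink eyelids from a delayed random toggle.
# Two RigidBody passes simulate ribbons and jaws with gravity override, weakened while skydiving.
# A final Control Rig blends down when the player moves, jumps, falls or aims.
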AnimBlueprint /BRCosmetics/Accessories/FORT_Backpacks/Backpack_F_MED_Scholar_Ghoul/Meshes/F_MED_Scholar_Ghoul_Pack_AnimBP
target_skeleton=&/F_MED_Scholar_Ghoul_Pack_Skeleton
graphs [
 AnimGraph {nodes [
   n1 = "Transform (Modify) Bone - Bone: dyn_large_mouth_jaw_jnt"(
    ComponentPose=Local To Component(LocalPose=Use cached pose 'ScholarGhoulCache'())
    Alpha=1
    Translation="0,0,0"
    Rotation="1,0,0"
    Scale="1,1,1"
   ) {
    rotationMode=BMM_Additive
    rotationSpace=BCS_BoneSpace
    translationMode=BMM_Additive
    translationSpace=BCS_BoneSpace
    alphaBoolBlend {blendOption=Linear}
    boneToModify {boneName=dyn_large_mouth_jaw_jnt}
    rotation {pitch=1}
   }
   n2 = "Transform (Modify) Bone - Bone: dyn_small_mouth_jaw_jnt"(ComponentPose=@n1, Alpha=1, Translation="0,0,0", Rotation="-4,0,0", Scale="1,1,1") {
    rotationMode=BMM_Additive
    rotationSpace=BCS_BoneSpace
    alphaBoolBlend {blendOption=Linear}
    boneToModify {boneName=dyn_small_mouth_jaw_jnt}
    rotation {pitch=-4}
   }
   n3 = "Transform (Modify) Bone - Bone: dyn_large_mouth_jnt"(ComponentPose=@n2, Alpha=Get Skydive_Float() -> (Skydive_Float:double), Translation="0,0,0", Rotation="0,15,0", Scale="1,1,1") {
    rotationMode=BMM_Additive
    rotationSpace=BCS_BoneSpace
    alphaBoolBlend {blendOption=Linear}
    boneToModify {boneName=dyn_large_mouth_jnt}
    rotation {yaw=15}
   }
   n4 = "Transform (Modify) Bone - Bone: dyn_small_mouth_jnt"(ComponentPose=@n3, Alpha=Get Skydive_Float(), Translation="0,0,0", Rotation="0,-15,0", Scale="1,1,1") {
    rotationMode=BMM_Additive
    rotationSpace=BCS_BoneSpace
    alphaBoolBlend {blendOption=Linear}
    boneToModify {boneName=dyn_small_mouth_jnt}
    rotation {yaw=-15}
   }
   n5 = Get Blink_Variable() -> (Blink_Variable:double)
   n6 = "Transform (Modify) Bone - Bone: dyn_lower_eyelid_rot_jnt"(ComponentPose=@n4, Alpha=@n5, Translation="0,0,0", Rotation="0,0,25", Scale="1,1,1") {
    rotationMode=BMM_Replace
    rotationSpace=BCS_BoneSpace
    alphaBoolBlend {blendOption=Linear}
    alphaScaleBiasClamp {bInterpResult=true, interpSpeedDecreasing=6, interpSpeedIncreasing=2}
    boneToModify {boneName=dyn_lower_eyelid_rot_jnt}
    rotation {roll=25}
   }
   n7 = "Transform (Modify) Bone - Bone: dyn_upper_eyelid_rot_jnt"(ComponentPose=@n6, Alpha=@n5, Translation="0,0,0", Rotation="0,0,-30", Scale="1,1,1") {
    rotationMode=BMM_Replace
    rotationSpace=BCS_BoneSpace
    alphaBoolBlend {blendOption=Linear}
    alphaScaleBiasClamp {bInterpResult=true, interpSpeedDecreasing=6, interpSpeedIncreasing=5}
    boneToModify {boneName=dyn_upper_eyelid_rot_jnt}
    rotation {roll=-30}
   }
   n8 = RigidBody!(
    ComponentPose=@n7
    Alpha=SelectFloat(A=0.7, B=1, bPickA=Get bIsSkyDiving())
    OverrideWorldGravity:Vector=Get GravityOverride()
    ExternalForce
   ) -> (Pose) {
    bClampLinearTranslationLimitToRefPose=true
    bOverrideWorldGravity=true
    overridePhysicsAsset=&/F_MED_Scholar_Ghoul_Pack_Ribbons_Physics
    simulationSpace=BaseBoneSpace
    alphaBoolBlend {blendOption=Linear}
    baseBoneRef {boneName=root}
    componentAppliedLinearAccClamp=<3000 3000 3000>
    componentLinearAccScale=<0.2 0.3 0.3>
    componentLinearVelScale=<0.2 0.3 0.3>
    overrideWorldGravity {z=-980}
   }
   n9 = RigidBody(ComponentPose=@n8, Alpha=SelectFloat(A=0.1, B=1, bPickA=Get bIsSkyDiving()), OverrideWorldGravity=Get GravityOverride()) {
    bClampLinearTranslationLimitToRefPose=true
    bOverrideWorldGravity=true
    overridePhysicsAsset=&/F_MED_Scholar_Ghoul_Pack_Jaws_Physics
    simulationSpace=BaseBoneSpace
    alphaBoolBlend {blendOption=Linear}
    baseBoneRef {boneName=root}
    componentAppliedLinearAccClamp=<3000 3000 3000>
    componentLinearAccScale=<0.2 0.3 0.3>
    componentLinearVelScale=<0.2 0.3 0.3>
    overrideWorldGravity {z=-980}
   }
   n10 = "float > float"(A=Get PawnSpeed2D(), B=1) {is=title K2Node_CallFunction}
   n11 = Get bIsTargeting()
   n12 = Get bIsJumping()
   n13 = Get bIsFalling()
   n14 = Control Rig(
    Source=Component To Local(ComponentPose=@n9)
    Alpha=SelectFloat(A=0, B=1, bPickA=OR Boolean!(A=@n12, B=@n13, C:bool=@n11, D:bool=@n10) -> (ReturnValue))
   ) {
    alphaBoolBlend {blendOption=Linear}
    alphaScaleBiasClamp {bInterpResult=true, bMapRange=true, interpSpeedDecreasing=5, outRange {min=0.2}}
    controlRigAssetReference {blueprintRigClass=&/F_MED_Scholar_Ghoul_Pack_CtrlRig.F_MED_Scholar_Ghoul_Pack_CtrlRig_C}
   }
   Output Pose(Result=@n14) {layerGroup=None}
   Save cached pose 'ScholarGhoulCache'(Pose=Local Space Ref Pose()) {cacheName=ScholarGhoulCache}
   OR Boolean!(A="float > float"(A=Get PawnSpeed2D(), B=1), B=OR Boolean(A=Get bIsFalling(), B=Get bIsJumping())) -> (ReturnValue)
  ]}
 EventGraph {nodes [
   Event BlueprintUpdateAnimation()
   > n1 = Gravity Override State Function DEPRECATED(CharacterPartAnimInstance=Self-Reference(), GameTimeInSeconds=GetGameTimeInSeconds(), GravityOverrideParams=Get Gravity Override Params())
   > Set GravityOverride(GravityOverride=@n1)
   n2 = Get bIsTargeting()
   > Set ADS(ADS:double=float - float(A=10, B="To Float (Boolean)"(InBool=@n2)) {is=title K2Node_CallFunction}) -> (Output_Get:double)
   n3 = Get bIsSkyDiving()
   > Set Skydiving(Skydiving:double=float - float(B="To Float (Boolean)"(InBool=@n3))) -> (Output_Get:double)
   > Set Skydive_Float(Skydive_Float:double=SelectFloat(A=1, B=0, bPickA=Get bIsSkydiveFloating())) -> (Output_Get:double)
   > Delay(Duration=2)
   > Set Blink_Variable(Blink_Variable:double="To Float (Boolean)"(InBool=RandomBool!() -> (ReturnValue:bool))) -> (Output_Get:double)
   > n4 = Sequence!(execute) -> (then_0, then_1, then_2:Exec)
   > Set Ext_Force(Ext_Force:Vector="0, 0,1250") -> (Output_Get:Vector)
   n5 = Get bIsFalling()
   n6 = Get bIsJumping()
   Set Jaw_Jump_Alpha(execute=@n4.then_1, Jaw_Jump_Alpha:double=SelectFloat(A=0.65, B=1, bPickA=OR Boolean(A=@n6, B=@n5))) -> (Output_Get:double)
   n7 = Get bIsJumping()
   Set Jaw_Jump_Alpha(execute=@n4.then_2, Jaw_Jump_Alpha=SelectFloat(A=0, B=1, bPickA=OR Boolean(A=@n7, B=false)))
   TryGetPawnOwner()
  ]}
]
variables [
 ADS {type=number}
 Blink_Variable {type=number}
 Ext_Force {type=Vector, default=<0 0 1250>}
 Gravity Override Params {type=GravityOverrideParamsStruct}
 GravityOverride {type=Vector}
 Jaw_Jump_Alpha {type=number}
 Skydive_Float {type=number}
 Skydiving {type=number}
]
